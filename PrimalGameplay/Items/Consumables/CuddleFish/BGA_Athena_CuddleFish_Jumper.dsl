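; Deployed Cuddle Fish proximity mine that leaps at pawns entering its jump trigger and that teammates can pick up.
; After landing and an arming delay, the server checks overlaps for hostility and line of sight, applying an LOS-checker effect when blocked and rechecking affiliation on health changes.
; A triggered jump replaces the mine with a thrown-fish projectile arced at the target's predicted position.
; Damage instead detonates it, applying an area explosion effect.
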(event Custom|SetMyTeam
  (params (Team :type Byte))
  ;[1 1 1] Set Team
  (Networking|SwitchHasAuthority
    (:Authority
      (bind _output_get (Variables|Default|SetMyTeam Team))
      (Variables|Default|SetTeamHasBeenSet true)
      (Development|PrintString (Utilities|String|Append "Set team to " (Utilities|String|ToString(Byte) _output_get)) true true "(R=1.000000,G=0.754107,B=0.000000,A=1.000000)" 5.0))))

(event Custom|OnInstigatorDied
  (params (DamagedActor :type Actor%20Object%20Reference) (Damage :type Float%20%28single-precision%29) (InstigatedBy :type Controller%20Object%20Reference) (DamageCauser :type Actor%20Object%20Reference) (HitLocation :type Vector) (FHitComponent :type Primitive%20Component%20Object%20Reference) (BoneName :type Name) (Momentum :type Vector))
  ;[1 1 1] OnInstigator Died Kill this mine
  (Actor|SetLifeSpan 0.01))

(event EventBeginPlay
  (|CallParentFunction)
  (Utilities|FlowControl|Sequence
    (:then_0
      (bind _asfort_pawn (Utilities|Casting|CastToFortPawn (Variables|Actor|GetInstigator)))
      (if (not (FortPawn|IsDead _asfort_pawn))
        (CallFunction|SetMyTeam (Team|GetActorTeam _asfort_pawn))
        (if (Utilities|SameTeam)
          (Components|Activation|Activate (Variables|Default|GetFriendlyVFX) true))))
    (:then_1
      ;[1 1 1] OnInstigator Died Kill this mine
      (if (Fortnite|IsCreativeorPlaygroundModePlaylist)
        (bind _asfort_player_pawn_athena (Utilities|Casting|CastToFortPlayerPawnAthena (Variables|Actor|GetInstigator)))
        (EventDispatchers|BindEventtoOnDied _asfort_player_pawn_athena (delegate Custom|OnInstigatorDied))))
    (:then_2
      (EventDispatchers|BindEventtoOnDied (delegate Custom|Destroyed)))
    (:then_3
      (Animation|Utilities|ForceTickAnimationforAllMeshes self)
      (Animation|Montage|PlayMontage (Variables|Default|GetSkeletalMesh) (Variables|Default|GetLandingMontage)))
    (:then_4
      (Fortnite|ShowSoundIndicatorLocal :Location (Transformation|GetActorLocation) :MaxAudibleDistance (Variables|/Script/Engine.SoundBase|GetMaxDistance (Variables|Sound|GetArmedSound)) :IndicatorType "Generic" :Instigator self :Tint "(R=0.000000,G=0.000000,B=0.000000,A=0.000000)" :OverrideIcon "/Game/UI/Foundation/Textures/Icons/HUD/T-Icon-Alert2-32.T-Icon-Alert2-32" :AttenuationMaxRangeMultiplier 1.0 :AttenuationIntensityMultiplier 1.0)
      ;[0 0.4521 1] Impact FX
      (Niagara|SpawnSystematLocation (Variables|VFX|GetLandedVFX) (Variables|/Script/Engine.SceneComponent|GetRelativeLocation (Variables|Default|GetStaticMesh)) (Math|Conversions|RotationFromXVector (Math|Vector|GetUpVector (Variables|/Script/Engine.SceneComponent|GetRelativeRotation (Variables|Default|GetStaticMesh)))))
      (Utilities|FlowControl|Delay (ScalableFloat|GetValueatLevel (Variables|Default|GetTimeToArm)))
      (Audio|PlaySoundatLocation (Variables|Sound|GetArmedSound) (Transformation|GetActorLocation))
      (Utilities|FlowControl|Sequence
        (:then_0
          ;[1 1 1] Initialize trigger radii
          (Components|Sphere|SetSphereRadius (Variables|Default|GetJumpTrigger) (ScalableFloat|GetValueatLevel (Variables|Default|GetJumpRadius)) true) ; NOTE: scale of root is set to 2 when BGA is spawned, so this radius is actually doubled.
        )
        (:then_1
          (if (<= (ScalableFloat|GetValueatLevel (Variables|Default|GetTimeToArm)) 0.1)
            ;[1 1 1] Arming after a Delay (need *at least* a 0.1 delay before turning on collision to avoid bugs)
            (Utilities|FlowControl|Delay 0.1)
            (else
              _))
          (Networking|SwitchHasAuthority
            (:Authority
              (Collision|SetCollisionEnabled (Variables|Default|GetJumpTrigger) "QueryOnly") ; Explosion Trigger isn't armed right away, gives enemy extra reaction time and prevents bugs
            )))))))

(event Custom|Destroyed
  (params (DamagedActor :type Actor%20Object%20Reference) (Damage :type Float%20%28single-precision%29) (InstigatedBy :type Controller%20Object%20Reference) (DamageCauser :type Actor%20Object%20Reference) (HitLocation :type Vector) (FHitComponent :type Primitive%20Component%20Object%20Reference) (BoneName :type Name) (Momentum :type Vector))
  ;[1 1 1] On Destroyed
  (Audio|Components|Audio|FadeOut (Variables|Default|GetAmbientHumAudio) 0.3 0.0 "Linear" :!class /Script/Engine.AudioComponent)
  (if (not (Variables|Default|GetIsExlode?))
    ;[1 1 1] Pick-Up effects
    (Audio|PlaySoundatLocation (Variables|Sound|GetPickUpSound) (Transformation|GetActorLocation)))
  (Rendering|SetActorHiddenInGame true)
  (Collision|SetActorEnableCollision false))

(event OnComponentBeginOverlap(JumpTrigger)
  (params (OverlappedComponent :type Primitive%20Component%20Object%20Reference) (OtherActor :type Actor%20Object%20Reference) (OtherComp :type Primitive%20Component%20Object%20Reference) (OtherBodyIndex :type Integer) (bFromSweep :type Boolean) (SweepResult :type Hit%20Result%20Structure%20%28by%20ref%29))
  ;[1 1 1] Explosion Triggers
  (Networking|SwitchHasAuthority
    (:Authority
      (Utilities|CheckforExplosionTrigger OtherActor
        (:Trigger
          (bind _data (Utilities|FlowControl|MergeExecutionwithData (Utilities|Casting|CastToFortPawn OtherActor) (stale Custom|CheckAffiliationChange _fortpawn))
            (:Then
              (bind _output_get (Variables|Default|SetJumpTarget _data))
              (Utilities|CheckLOS 3e+01 _output_get
                (:LOS
                  (CallFunction|HandleJump))
                (:No%20LOS
                  (GameplayEffects|ApplyGameplayEffectToTarget :self (Ability|GetAbilitySystemComponent self) :GameplayEffectClass "/Game/Athena/Items/Weapons/Prototype/PurpleMouse/GE_PurpleMouse_ApplyLOSChecker.GE_PurpleMouse_ApplyLOSChecker_C" :Target (Ability|GetAbilitySystemComponent _output_get) :Level 0.0))))))
        (:No%20Trigger
          ;;[1 1 1] Bind and unbind on health change to see if affiliation changed
          (bind _asplayer_pawn_athena (Utilities|Casting|CastToPlayerPawn_Athena OtherActor))
          (Health|AssignOnPawnHealthChanged _asplayer_pawn_athena (EventDispatchers|CreateEvent :!event Custom|CheckAffiliationChange)))))))

(event EventStopSim
  (params (Hit :type Hit%20Result%20Structure))
  (bind (_bblockinghit _binitialoverlap _time _distance _location _impactpoint _normal) (Collision|BreakHitResult Hit))
  ;[1 1 1] Stop Sim
  (|CallParentFunction Hit)
  (Variables|Default|SetHasLanded true)
  (Variables|Default|SetIsFalling false)
  (if (and (not (Variables|Default|GetJumped)) (> (Math|Vector|VectorLength _normal) 0.0))
    (Transformation|SetActorLocation _impactpoint false true)
    (if (Math|Vector|NotEqualExactly(Vector) _normal)
      (if (Variables|Default|GetHasAttached)
        (Transformation|SetActorRotation (Math|Rotator|MakeRotfromZ _normal) true)
        (Niagara|SpawnSystematLocation (Variables|VFX|GetLandedVFX) (Variables|/Script/Engine.SceneComponent|GetRelativeLocation (Variables|Default|GetStaticMesh)) (Math|Conversions|RotationFromXVector (Math|Vector|GetUpVector (Variables|/Script/Engine.SceneComponent|GetRelativeRotation (Variables|Default|GetStaticMesh)))))))))

(event Custom|CheckAffiliationChange
  ;[1 1 1] Explosion Triggers
  (Networking|SwitchHasAuthority
    (:Authority
      (bind (_hostilefound _fortpawn) (CallFunction|CheckforHostilePawn))
      (if _hostilefound
        (bind _data (Utilities|FlowControl|MergeExecutionwithData (Utilities|Casting|CastToFortPawn (stale OnComponentBeginOverlap(JumpTrigger) OtherActor)) _fortpawn)
          (:Then
            (bind _output_get (Variables|Default|SetJumpTarget _data))
            (Utilities|CheckLOS 3e+01 _output_get
              (:LOS
                (CallFunction|HandleJump))
              (:No%20LOS
                (GameplayEffects|ApplyGameplayEffectToTarget :self (Ability|GetAbilitySystemComponent self) :GameplayEffectClass "/Game/Athena/Items/Weapons/Prototype/PurpleMouse/GE_PurpleMouse_ApplyLOSChecker.GE_PurpleMouse_ApplyLOSChecker_C" :Target (Ability|GetAbilitySystemComponent _output_get) :Level 0.0)))))))))

(event OnComponentEndOverlap(JumpTrigger)
  (params (OverlappedComponent :type Primitive%20Component%20Object%20Reference) (OtherActor :type Actor%20Object%20Reference) (OtherComp :type Primitive%20Component%20Object%20Reference) (OtherBodyIndex :type Integer))
  ;[1 1 1] Explosion Triggers
  (GameplayEffects|RemoveActiveEffectswithTags (Ability|GetAbilitySystemComponent OtherActor) "(GameplayTags=((TagName=\"Item.Consumable.PurpleMouse.LOSChecker\")))")
  ;;[1 1 1] Bind and unbind on health change to see if affiliation changed
  (bind _asplayer_pawn_athena (Utilities|Casting|CastToPlayerPawn_Athena OtherActor))
  (Health|UnbindEventfromOnPawnHealthChanged _asplayer_pawn_athena (EventDispatchers|CreateEvent :!event Custom|CheckAffiliationChange)))

(event EventOnDamageServer
  (params (Damage :type Float%20%28single-precision%29) (DamageTags :type Gameplay%20Tag%20Container%20Structure%20%28by%20ref%29) (Momentum :type Vector :default "0, 0, 0") (HitInfo :type Hit%20Result%20Structure%20%28by%20ref%29) (InstigatedBy :type Controller%20Object%20Reference) (DamageCauser :type Actor%20Object%20Reference) (EffectContext :type Gameplay%20Effect%20Context%20Handle%20Structure))
  ;[1 1 1] Explosion Triggers
  (CallFunction|Explode))

(event OnEnterWaterEvent(FortWaterInteraction)
  (params (WaterBody :type Fort%20Water%20Body%20Actor%20Object%20Reference) (WaterInteractionComponent :type Fort%20Water%20Interaction%20Component%20Object%20Reference) (bIsFirstBody :type Boolean)))

(event Custom|Explode
  (bind _instigator (Variables|Actor|GetInstigator))
  (bind _returnvalue (Ability|GetFortAbilitySystemComponent _instigator))
  ;[1 1 1] Explode
  (Utilities|FlowControl|Sequence
    (:then_0
      (Components|DestroyComponent (Variables|Default|GetBeepParticles))
      (Collision|SetActorEnableCollision false)
      (Actor|SetLifeSpan 0.01)
      (CallFunction|SpawnExplosionParticles)
      (GameplayCue|ExecuteGameplayCue%28Burst%29Local self (Variables|Default|GetCameraShakeGC) (Ability|GameplayCue|MakeGameplayCueParameters :NormalizedMagnitude 1.0 :RawMagnitude 3.0 :Location (Transformation|GetActorLocation) :Normal "0,0,0" :GameplayEffectLevel 1 :AbilityLevel 1 :bReplicateLocationWhenUsingMinimalRepProxy false))
      (Fortnite|ShowSoundIndicatorLocalforTeam :Location (Transformation|GetActorLocation) :MaxAudibleDistance (Variables|/Script/Engine.SoundBase|GetMaxDistance (Variables|Sound|GetExplodeSound)) :IndicatorType "Gunshot" :Team (Variables|Default|GetMyTeam) :Affiliations (Utilities|Array|MakeArray "Hostile" "Neutral") :Tint "(R=0.000000,G=0.000000,B=0.000000,A=0.000000)" :AttenuationMaxRangeMultiplier 1.0 :AttenuationIntensityMultiplier 1.0))
    (:then_1
      (Utilities|FlowControl|DoOnce false
        (:Completed
          (bind _outactors (Collision|SphereOverlapActors (Transformation|GetActorLocation) (ScalableFloat|GetValueatLevel (Variables|Default|GetDamageRadius)) (Utilities|Array|MakeArray "ObjectTypeQuery6" "ObjectTypeQuery1" "ObjectTypeQuery5" "ObjectTypeQuery3" "ObjectTypeQuery2" "ObjectTypeQuery4" "ObjectTypeQuery7") "/Script/Engine.Actor" (Utilities|Array|MakeArray self)))
          (bind _result (CallFunction|FilterbyLOS _outactors))
          (Utilities|IsValid _returnvalue
            (:Is%20Valid
              (Variables|Default|SetDamageDealingAbilitySystemComponent _returnvalue))
            (:Is%20Not%20Valid
              (bind _asfort_ability_system_component (Utilities|Casting|CastToFortAbilitySystemComponent (Ability|GetAbilitySystemComponent self)))
              (Variables|Default|SetDamageDealingAbilitySystemComponent _asfort_ability_system_component)))
          (for _array_element _result
            (bind _asfort_pawn (Utilities|Casting|CastToFortPawn _array_element)
              (:then
                (if (or (Utilities|Equal(Object) _asfort_pawn (Variables|Actor|GetInstigator)) (Math|Byte|NotEqual(Byte) (Team|GetActorTeam _asfort_pawn) (Variables|Default|GetMyTeam)))
                  _))
              (:CastFailed
                _))
            (bind _returnvalue_2 (Utilities|MakeEffectContext))
            (GameplayEffects|ApplyGameplayEffectToTarget (Variables|Default|GetDamageDealingAbilitySystemComponent) (Variables|Default|GetExplosionGE) (Ability|GetAbilitySystemComponent _array_element) 0.0 _returnvalue_2)))))))

(event Custom|SpawnExplosionParticles
  (bind _returnvalue (Transformation|GetActorLocation))
  ;[1 1 1] Explode
  (Networking|CanExecuteCosmeticEvents
    (:True
      (Audio|PlaySoundatLocation (Variables|Sound|GetExplodeSound) _returnvalue "0, 0, 0" 1.0 1.0 0.0 0 0 self)
      (bind _returnvalue_2 (Niagara|SpawnSystematLocation (Variables|VFX|GetExplosionVFX) _returnvalue))
      (Rendering|SetReceivesDecals _returnvalue_2 false)
      (GameplayCue|ExecuteGameplayCue%28Burst%29Local self (Variables|Default|GetCameraShakeGC) (Ability|GameplayCue|MakeGameplayCueParameters 1.0 3.0))
      (Fortnite|ShowSoundIndicatorLocal 0 (Transformation|GetActorLocation) (Variables|/Script/Engine.SoundBase|GetMaxDistance (Variables|Default|GetExplodeSFX)) "Gunshot")
      (Game|SpawnActorfromClass "/PrimalGameplay/Items/Consumables/CuddleFish/VFX/B_Decal_CuddleFish.B_Decal_CuddleFish_C" (Math|Conversions|ToTransform(Vector) (Transformation|GetActorLocation)) "Undefined" "MultiplyWithRoot" 0 0))))

(event Custom|HandleJump
  ;[1 1 1] Jump
  (Variables|Default|SetJumpTriggered true)
  (Utilities|FlowControl|DoOnce false
    (:Completed
      (Networking|CanExecuteCosmeticEvents
        (:True
          _)
        (:False
          _))
      (Utilities|FlowControl|Delay (ScalableFloat|GetValueatLevel (Variables|Default|GetJumpDelay)))
      (Networking|CanExecuteCosmeticEvents
        (:True
          (Niagara|SpawnSystematLocation "/PrimalGameplay/Items/Consumables/CuddleFish/VFX/NS_CuddleFish_GroundHit.NS_CuddleFish_GroundHit" (+ (Transformation|GetActorLocation) "0, 0,10.000000") (Math|Conversions|RotationFromXVector (Variables|Default|GetGravHitNormal)))
          (Audio|SpawnSoundAttached (Variables|Sound|GetJumpTriggeredSound) (Variables|Default|GetSkeletalMesh)))
        (:False
          (CallFunction|DetachandLaunch))))))

(event OnProjectileStop(ProjectileMovement)
  (params (ImpactResult :type Hit%20Result%20Structure%20%28by%20ref%29))
  ;[1 1 1] Jump
  (Utilities|FlowControl|DoOnce false
    (:Completed
      (Networking|CanExecuteCosmeticEvents
        (:True
          _)
        (:False
          _))
      (Utilities|FlowControl|Delay (ScalableFloat|GetValueatLevel (Variables|Default|GetJumpDelay)))
      (Networking|CanExecuteCosmeticEvents
        (:True
          (Niagara|SpawnSystematLocation "/PrimalGameplay/Items/Consumables/CuddleFish/VFX/NS_CuddleFish_GroundHit.NS_CuddleFish_GroundHit" (+ (Transformation|GetActorLocation) "0, 0,10.000000") (Math|Conversions|RotationFromXVector (Variables|Default|GetGravHitNormal)))
          (Audio|SpawnSoundAttached (Variables|Sound|GetJumpTriggeredSound) (Variables|Default|GetSkeletalMesh)))
        (:False
          (CallFunction|DetachandLaunch))))))

(event Custom|DetachAndLaunch
  (bind _returnvalue (Transformation|GetActorUpVector))
  (bind _returnvalue_2 (* _returnvalue 3e+01))
  (bind _returnvalue_3 (Transformation|GetActorLocation))
  (bind _returnvalue_4 (+ _returnvalue_2 _returnvalue_3))
  ;[1 1 1] Jump
  (bind _outlaunchvelocity (Game|SuggestProjectileVelocityCustomArc _returnvalue_4 (+ (FortPawn|GetMovementPredictionPosition (Variables|Default|GetJumpTarget) (ScalableFloat|GetValueatLevel (Variables|Default|GetPredictionDelta))) (* (Transformation|GetActorUpVector (Variables|Default|GetJumpTarget)) (* (Components|Capsule|GetScaledCapsuleHalfHeight (Variables|/Script/Engine.Character|GetCapsuleComponent (Variables|Default|GetJumpTarget))) 0.5)))))
  (bind _returnvalue_5 (Projectile|SpawnProjectile :ProjectileClass "/PrimalGameplay/Items/Consumables/CuddleFish/B_Prj_Athena_Thrown_CuddleFish.B_Prj_Athena_Thrown_CuddleFish_C" :RequestedBy (Variables|Actor|GetInstigator) :SpawnLocation _returnvalue_4 :SpawnRotation (Transformation|GetActorRotation) :SpawnDirection (Math|Conversions|RotationFromXVector _outlaunchvelocity) :bUseDefaultPhysics false :InitialSpeed (Math|Vector|VectorLength _outlaunchvelocity) :GravityScale 1.0 :ChargePercent 0.0)) ; Spawn Cuddle Fish Projectile
  (bind _asb_prj_athena_thrown_cuddle_fish (Utilities|Casting|CastToB_Prj_Athena_Thrown_CuddleFish _returnvalue_5))
  (Class|BPrjAthenaThrownCuddleFish|SetHasSpawnedfromJump _asb_prj_athena_thrown_cuddle_fish)
  (Actor|SetLifeSpan 0.01))

(event Interaction|EventBlueprintOnInteract
  (params (InteractingPawn :type Fort%20Pawn%20Object%20Reference) (InteractionBeingAttempted :type EInteractionBeingAttempted%20Enum :default "FirstInteraction"))
  (bind _asfort_player_controller (Utilities|Casting|CastToFortPlayerController (Pawn|GetController InteractingPawn)))
  (Inventory|GiveItemtoInventoryOwner _asfort_player_controller (Variables|Default|GetItemToPickUpOnDisarm))
  (Variables|Default|SetIsInteract? true)
  (Actor|DestroyActor))

(event EventSpawnFXSounds
  (|CallParentFunction)
  (Audio|Components|Audio|FadeIn (Variables|Default|GetAmbientHumAudio) 0.25 1.0 0.0 "Linear" :!class /Script/Engine.AudioComponent))

(event EventRestartSimulation
  (|CallParentFunction)
  (Variables|Default|SetIsFalling true)
  (Variables|Default|SetHasLanded false))

(fn AttachToBindedActor
  (params (AttachComp :type Primitive%20Component%20Object%20Reference))
  (|CallParentFunction AttachComp)
  (Variables|Default|SetHasAttached true))

(fn BlueprintCanInteract
  (params (InteractingPawn :type Fort%20Pawn%20Object%20Reference) (InteractionBeingAttempted :type EInteractionBeingAttempted%20Enum :default "FirstInteraction") (InteractionType :type TInteractionType%20Enum :default "IT_NoInteraction"))
  (return (Math|Boolean|ANDBoolean (Math|Byte|Equal(Byte) (Team|GetActorTeam InteractingPawn) (Team|GetActorTeam (Variables|Actor|GetInstigator))) (not (GameplayTags|HasAnyTags (Class|FortGameStateAthena|GetAthenaPlaylistContextTags (Fortnite|Athena|GameState|GetGameStateAthena)) (LocalVariables|GetSpyLTMFactionTag))) (not (Variables|Default|GetJumped)) (not (GameplayTags|HasAllTags (GameplayTags|GetOwnedGameplayTags InteractingPawn :!class /Script/GameplayTags.GameplayTagAssetInterface) (LocalVariables|GetAimingWithCuddleFish))) (not (Variables|Default|GetJumpTriggered)))))

(fn BlueprintGetInteractionString
  (params (InteractingPawn :type Fort%20Pawn%20Object%20Reference) (InteractionBeingAttempted :type EInteractionBeingAttempted%20Enum :default "FirstInteraction"))
  (return (Variables|Default|GetInteractText)))

(fn BlueprintGetInteractionTime
  (params (InteractingPawn :type Fort%20Pawn%20Object%20Reference) (InteractionBeingAttempted :type EInteractionBeingAttempted%20Enum :default "FirstInteraction"))
  (return 0.0 true))

(fn CanBeSavedToCreativeVolume
  (return false))

(fn CheckForHostilePawn
  (bind _array_element (Utilities|Array|ForEachLoopwithBreak (Collision|GetOverlappingActors (Variables|Default|GetJumpTrigger) "/Game/Athena/PlayerPawn_Athena.PlayerPawn_Athena_C" :!class /Script/Engine.PrimitiveComponent))
    (:LoopBody
      (Utilities|CheckforExplosionTrigger _array_element
        (:Trigger
          (bind _asfort_pawn (Utilities|Casting|CastToFortPawn _array_element))
          (LocalVariables|SetLocalPawn _asfort_pawn))))
    (:Completed
      _))
  (bind _localpawn (LocalVariables|GetLocalPawn)
    (:then
      (return true _localpawn))
    (:else
      (return false 0))))

(fn FilterByLOS
  (params (_Array :type Array%20of%20Actor%20Object%20References))
  (for _array_element _Array
    (bind _asfort_pawn (Utilities|Casting|CastToFortPawn _array_element)
      (:then
        (Utilities|CheckLOS 0 _asfort_pawn
          (:LOS
            _)))
      (:CastFailed
        _))
    (Utilities|Array|Add (LocalVariables|GetResult) _array_element))
  (return (LocalVariables|GetResult)))

(fn GetAllPawns
  (params (_Array :type Array%20of%20Actor%20Object%20References))
  (for _array_element _Array
    (Utilities|Casting|CastToFortPawn _array_element)
    (Utilities|Array|Add (LocalVariables|GetResult) _array_element))
  (return (LocalVariables|GetResult)))

(fn GetFocusedSocketLocation
  (return (+ (Transformation|GetActorLocation) "0, 0,30.000000")))

(fn HandleNPC_AI
  (params (NPC%20Pawn :type NPC%20Pawn%20Parent%20Object%20Reference) (Success :type Boolean))
  (bind (_hasleader _leader) (Class|NPCPawnParent|NPCHasLeader? NPC%20Pawn))
  (if Success
    (return (or (Math|Boolean|ANDBoolean Success _hasleader (Math|Byte|NotEqual(Byte) (Team|GetActorTeam _leader) (Variables|Default|GetMyTeam))) (Math|Boolean|ANDBoolean (not _hasleader) (Utilities|Enum|Equal(Enum) (Variables|/Game/Athena/AI/Pawns/NPC/Parents/NPC_Pawn_Parent.NPC_Pawn_Parent_C|GetNPC_AlertLevel NPC%20Pawn) "NewEnumerator3") (Math|Byte|Equal(Byte) (Team|GetActorTeam (Goals|GetCurrentAimTarget NPC%20Pawn)) (Variables|Default|GetMyTeam)))))
    (else
      (return false))))

(fn HandleWildlife
  (params (WildlifePawn :type NPC%20Pawn%20Wildlife%20Parent%20Object%20Reference) (Success :type Boolean))
  (bind (_hasleader _leader) (Class|NPCPawnParent|NPCHasLeader? WildlifePawn))
  (if Success
    (return (or (Math|Boolean|ANDBoolean Success _hasleader (Math|Byte|NotEqual(Byte) (Team|GetActorTeam _leader) (Variables|Default|GetMyTeam))) (Math|Boolean|ANDBoolean (not _hasleader) (Utilities|Enum|Equal(Enum) (Variables|/Game/Athena/AI/Pawns/NPC/Parents/NPC_Pawn_Parent.NPC_Pawn_Parent_C|GetNPC_AlertLevel WildlifePawn) "NewEnumerator3") (Math|Byte|Equal(Byte) (Team|GetActorTeam (Goals|GetCurrentAimTarget WildlifePawn)) (Variables|Default|GetMyTeam)))))
    (else
      (return false))))

(fn Init
  (params (GravHitNormal :type Vector :default "0, 0, 0"))
  (|CallParentFunction GravHitNormal)
  (Transformation|SetActorRotation (Math|Rotator|MakeRotfromZ (Variables|Default|GetGravHitNormal)) true))

(fn IsinInfiltrationMode
  (return (GameplayTags|HasTag (Class|FortGameStateAthena|GetAthenaPlaylistContextTags (Fortnite|Athena|GameState|GetGameStateAthena)) "(TagName=\"Athena.Playlist.Infiltration\")")))

(fn ConstructionScript
  (Networking|SwitchHasAuthority
    (:Authority
      (Variables|/Script/Engine.PrimitiveComponent|SetbGenerateOverlapEvents true (Variables|Default|GetJumpTrigger))
      (|CallParentFunction))))
