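; Despawn actor for the Cuddle Fish consumable, making it appear to jump off the player it damaged with ink.
; BeginPlay launches the skeletal mesh via projectile movement, upward and away from any attached actor with slight random yaw, and tags it "Jumped".
; Projectile stop adds an "AttachedToPawn" tag for the animation Blueprint and plays a landing sound.
; After a hotfixable lifetime delay, a vanish montage plays and the actor destroys itself on completion.
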(event EventBeginPlay
  (bind _fortprojectilemovement (Variables|Default|GetFortProjectileMovement))
  (bind _skeletalmesh (Variables|Default|GetSkeletalMesh))
  ;[0 0.15 0.005892 0.5] BeginPlay: Execute behavior so it looks like the Cuddle Fish is jumping off the player that it did damage to by squirting ink.
  (|CallParentFunction)
  (bind _attachedactor (Variables|Default|GetAttachedActor)
    (:then
      (Variables|SetJumpVelocity_X (Math|Vector|MakeVector (.x (* (Math|Vector|Normalize2D(Vector) (Math|Vector|GetUnitDirection(Vector) (Transformation|GetActorLocation _attachedactor) (Transformation|GetActorLocation))) 1e+02)) (.y (* (Math|Vector|Normalize2D(Vector) (Math|Vector|GetUnitDirection(Vector) (Transformation|GetActorLocation _attachedactor) (Transformation|GetActorLocation))) 1e+02)) 5e+02))
      (Transformation|SetActorRotation (Math|Rotator|MakeRotator 0.0 0.0 (+ (.yaw (Math|Rotator|FindLookatRotation (Transformation|GetActorLocation (Variables|Default|GetAttachedActor)) (Transformation|GetActorLocation))) (Math|Random|RandomFloatinRange -1e+01 1e+01)))))
    (:else
      (Variables|SetJumpVelocity_X (0.0 0.0 5e+02))))
  (Components|Movement|SetUpdatedComponent _fortprojectilemovement (Variables|Default|GetSkeletalMesh))
  (Projectile|SetVelocity _fortprojectilemovement (Variables|Transient|GetJumpVelocity))
  (Components|SkeletalMesh|SetAnimInstanceClass _skeletalmesh "/PrimalGameplay/Items/Consumables/CuddleFish/Assets/Animation/CuddleFish_AnimBP.CuddleFish_AnimBP_C")
  (bind _ascuddle_fish_anim_bp (Utilities|Casting|CastToCuddleFish_AnimBP (Components|SkeletalMesh|GetAnimInstance _skeletalmesh))
    (:then
      (Variables|Transient|SetCuddleFishAnimBlueprint _ascuddle_fish_anim_bp)
      (Utilities|Array|Add (Variables|Actor|GetTags) "Jumped"))
    (:CastFailed
      _))
  (Utilities|FlowControl|Delay (ScalableFloat|GetValueatLevel (Variables|Default|GetHotfix_LifetimeDuration)))
  (Animation|Montage|PlayMontage (Variables|Default|GetSkeletalMesh) "/PrimalGameplay/Items/Consumables/CuddleFish/Assets/Animation/CuddleFish_Despawn_Vanish_CF_Montage.CuddleFish_Despawn_Vanish_CF_Montage" 1.0 0.0 ""
    (:OnCompleted
      (Actor|DestroyActor))))

(event OnProjectileStop(FortProjectileMovement)
  (params (ImpactResult :type Hit%20Result%20Structure%20%28by%20ref%29))
  ;[0.15 0.1023 0.01806 0.5] OnProjectileStopped: Add tag so it can be picked up by the animation blueprint.
  (Utilities|Array|Add (Variables|Actor|GetTags) "AttachedToPawn")
  (Audio|PlaySoundatLocation (Variables|Default|GetLandedOnEnvironment) (Transformation|GetWorldLocation (Variables|Default|GetSkeletalMesh))))

(fn ConstructionScript)
