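; Thrown Cuddle Fish projectile that sticks to pawns and explodes, or spawns a Cuddle Fish Jumper on surface hits.
; Server hit and overlap checks drive pawn attachment or BGA spawning, passing the instigator's team to the jumper.
; Attaching or stopping starts a montage-timed explosion, applying a gameplay effect to line-of-sight-filtered overlaps plus the attached actor, unless a BGA spawned.
; Death or destruction of the attached actor restarts simulation; actor tags mark animation states.
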(event EventBeginPlay
  ;[1 1 1] Set Team and Self Stick Delay
  (Utilities|FlowControl|Sequence
    (:then_0
      (if (not (Variables|BGASpawning|GetTeamHasBeenSet))
        (bind _asfort_pawn (Utilities|Casting|CastToFortPawn (Game|GetInstigator))
          (:then
            (if (not (FortPawn|IsDead _asfort_pawn))
              (CallFunction|SetTeam (Team|GetActorTeam _asfort_pawn)))
            (|CallParentFunction))
          (:CastFailed
            _))))
    (:then_1
      (Utilities|FlowControl|Delay (Variables|AttachableProjectile|GetSelfStickDelay))
      (Variables|AttachableProjectile|SetIsSelfStickDelayOver? true))
    (:then_2
      (Utilities|Time|SetTimerbyEvent (EventDispatchers|CreateEvent :!event Custom|CheckVelocity) 0.1 true))))

(event Custom|SetTeam
  (params (Team :type Byte))
  ;[1 1 1] Set Team and Self Stick Delay
  (Variables|BGASpawning|SetTeam Team)
  (Variables|BGASpawning|SetTeamHasBeenSet true))

(event OnEnterWaterEvent(WaterInteractionComponent)
  (params (WaterBody :type Fort%20Water%20Body%20Actor%20Object%20Reference) (WaterInteractionComponent :type Fort%20Water%20Interaction%20Component%20Object%20Reference) (bIsFirstBody :type Boolean))
  ;[1 1 1] Spawn on Enter Water
  (Utilities|FlowControl|Sequence
    (:then_0
      (if bIsFirstBody
        (CallFunction|HeldWaterInteraction)))
    (:then_1
      (Networking|CanExecuteCosmeticEvents
        (:True
          (bind _outwatersurfaceinfo (WaterBody|GetWaterSurfaceInfoStruct WaterBody (Transformation|GetActorLocation) true false false))
          (bind _returnvalue (Effects|SpawnEmitteratLocation (Variables|VFX|GetFX_HitWater) (Utilities|Struct|BreakWaterSurfaceInfo _outwatersurfaceinfo) (Math|Rotator|MakeRotfromX (.WaterPlaneNormal (Utilities|Struct|BreakWaterSurfaceInfo _outwatersurfaceinfo))) "1.000000,1.000000,1.000000" true "AutoRelease"))
          (Effects|Components|ParticleSystem|SetFloatParameter _returnvalue "WaterDepth" (.WaterDepth (Utilities|Struct|BreakWaterSurfaceInfo _outwatersurfaceinfo)))
          (Effects|Components|ParticleSystem|SetFloatParameter _returnvalue "WaterBodyIndex" (Math|Conversions|ToFloat(Integer) (.WaterBodyIdx (Utilities|Struct|BreakWaterSurfaceInfo _outwatersurfaceinfo)))))))))

(event Custom|HeldWaterInteraction
  ;[1 1 1] Spawn on Enter Water
  (Audio|PlaySoundatLocation (Variables|Sound|GetSplashSound) (Transformation|GetActorLocation)))

(event Collision|EventHit
  (params (MyComp :type Primitive%20Component%20Object%20Reference) (Other :type Actor%20Object%20Reference) (OtherComp :type Primitive%20Component%20Object%20Reference) (bSelfMoved :type Boolean) (HitLocation :type Vector :default "0, 0, 0") (HitNormal :type Vector :default "0, 0, 0") (NormalImpulse :type Vector :default "0, 0, 0") (Hit :type Hit%20Result%20Structure%20%28by%20ref%29))
  (bind (_bblockinghit _binitialoverlap _time _distance _location _impactpoint _normal _impactnormal _physmat _hitactor _hitcomponent _hitbonename _bonename _hititem _elementindex _faceindex _tracestart _traceend) (Collision|BreakHitResult Hit))
  (bind (_asfort_athena_vehicle _bsuccess) (Utilities|Casting|CastToFortAthenaVehicle Other))
  ;[1 1 1] EventOnTouched
  (|CallParentFunction MyComp Other OtherComp bSelfMoved HitLocation HitNormal NormalImpulse Hit)
  (Networking|SwitchHasAuthority
    (:Authority
      (if (and (not (Components|ComponentHasTag OtherComp "NoStick")) (Variables|BGASpawning|GetSpawnOnCollision))
        (Utilities|Casting|CastToBGA_Athena_WithGravity_Parent Other
          (:CastFailed
            (bind _returnvalue (GameplayTags|HasMatchingGameplayTag(Message) Other "(TagName=\"HeldObject\")"))
            (bind _returnvalue_2 (GameplayTags|HasMatchingGameplayTag(Message) Other "(TagName=\"Pawn.Athena.NPC.Wildlife\")"))
            (if (not (and (not _returnvalue_2) _returnvalue))
              (bind _asfort_pawn (Utilities|Casting|CastToFortPawn Other)
                (:then
                  (Utilities|Casting|CastToFortCharacterVehicle_Ostrich _asfort_pawn
                    (:then
                      (CallFunction|HandleHits Other OtherComp))
                    (:CastFailed
                      (Utilities|SticktoPawnCheck _asfort_pawn
                        (:True
                          (Variables|AttachableProjectile|SetStuckToPawn _asfort_pawn)
                          (CallFunction|AttachtoActor Hit))))))
                (:CastFailed
                  (if (not (and (Variables|BGASpawning|GetDontSpawnOnVehicles) _bsuccess))
                    (Game|Components|ProjectileMovement|StopSimulating (Variables|Movement|GetProjectileMovementComponent) Hit)
                    (Utilities|FlowControl|DoOnce false
                      (:Completed
                        (CallFunction|SpawnBGA _normal _hitactor _hitcomponent (CallFunction|GetClosestPoint OtherComp (Transformation|GetActorLocation)) false (- _traceend _tracestart))))))))))))))

(event OnComponentBeginOverlap(CapsuleComponent)
  (params (OverlappedComponent :type Primitive%20Component%20Object%20Reference) (OtherActor :type Actor%20Object%20Reference) (OtherComp :type Primitive%20Component%20Object%20Reference) (OtherBodyIndex :type Integer) (bFromSweep :type Boolean) (SweepResult :type Hit%20Result%20Structure%20%28by%20ref%29))
  ;[1 1 1] Handle Larger Pawn Overlap Sphere
  (if (Utilities|NotEqual(Object) OtherActor (Variables|Actor|GetInstigator))
    _
    (elif (Variables|AttachableProjectile|GetIsSelfStickDelayOver?)
      _))
  (Networking|SwitchHasAuthority
    (:Authority
      (bind _asfort_pawn (Utilities|Casting|CastToFortPawn OtherActor)
        (:then
          (Utilities|SticktoPawnCheck _asfort_pawn
            (:True
              (Variables|AttachableProjectile|SetStuckToPawn _asfort_pawn)
              (CallFunction|AttachtoActor (stale Collision|EventHit Hit)))))
        (:CastFailed
          (bind _returnvalue (GameplayTags|HasAnyMatchingGameplayTags(Message) OtherActor (Variables|AttachableProjectile|GetTC_ActorTagsThatShouldExplodeOnOverlap)))
          (if _returnvalue
            (CallFunction|CallExplodeTrigger)))))))

(event Custom|AttachToActor
  (params (Hit :type Hit%20Result%20Structure))
  ;[1 1 1] Gate attaching to player pawns
  (Utilities|FlowControl|Gate false
    (:Exit
      (Utilities|FlowControl|DoOnce false
        (:Completed
          (Gameplay|ServerAttachtoActor Hit (Variables|AttachableProjectile|GetStuckToPawn) (Variables|AttachableProjectile|GetOffsetDistanceFromPhysicsMesh) (Variables|AttachableProjectile|GetOffsetDistanceFromBone) 0.75)
          (CallFunction|AttachtoActorVFX)
          (Transformation|SetRelativeRotation (Variables|Default|GetSkeletalMesh) "0, 0, 0" false false)
          (CallFunction|BindtoPlayerorAIDeath)
          (CallFunction|CallExplodeTrigger))))))

(event Custom|ReopenPawnCollision
  ;[1 1 1] Gate attaching to player pawns
  (Utilities|FlowControl|Sequence
    (:then_0
      (Utilities|FlowControl|Gate false
        (:Exit
          _)))
    (:then_1
      _)
    (:then_2
      ;[1 1 1] OnStop - Bind destruction of building actor if relevant
      (Utilities|FlowControl|DoOnce false
        (:Completed
          (Transformation|AttachActorToComponent (.HitComponent (Collision|BreakHitResult (stale OnProjectileStop(ProjectileMovementComponent) ImpactResult))) "None" "KeepWorld" "KeepWorld" "KeepWorld" false)))))
  (Utilities|FlowControl|DoOnce false
    (:Completed
      (Gameplay|ServerAttachtoActor (stale Custom|AttachToActor Hit) (Variables|AttachableProjectile|GetStuckToPawn) (Variables|AttachableProjectile|GetOffsetDistanceFromPhysicsMesh) (Variables|AttachableProjectile|GetOffsetDistanceFromBone) 0.75)
      (CallFunction|AttachtoActorVFX)
      (Transformation|SetRelativeRotation (Variables|Default|GetSkeletalMesh) "0, 0, 0" false false)
      (CallFunction|BindtoPlayerorAIDeath)
      (CallFunction|CallExplodeTrigger))))

(event Custom|ClosePawnCollision
  ;[1 1 1] Gate attaching to player pawns
  (Utilities|FlowControl|Gate false
    (:Exit
      (Utilities|FlowControl|DoOnce false
        (:Completed
          (Gameplay|ServerAttachtoActor (stale Custom|AttachToActor Hit) (Variables|AttachableProjectile|GetStuckToPawn) (Variables|AttachableProjectile|GetOffsetDistanceFromPhysicsMesh) (Variables|AttachableProjectile|GetOffsetDistanceFromBone) 0.75)
          (CallFunction|AttachtoActorVFX)
          (Transformation|SetRelativeRotation (Variables|Default|GetSkeletalMesh) "0, 0, 0" false false)
          (CallFunction|BindtoPlayerorAIDeath)
          (CallFunction|CallExplodeTrigger))))))

(event OnProjectileStop(ProjectileMovementComponent)
  (params (ImpactResult :type Hit%20Result%20Structure%20%28by%20ref%29))
  (bind (_bblockinghit _binitialoverlap _time _distance _location _impactpoint _normal _impactnormal _physmat _hitactor _hitcomponent) (Collision|BreakHitResult ImpactResult))
  ;[1 1 1] OnStop - Bind destruction of building actor if relevant
  (Variables|AttachableProjectile|SetbSimulationRunning false)
  (Variables|AttachableProjectile|SetHasStopped true)
  (Variables|AttachableProjectile|SetHitNormal _impactnormal)
  (Utilities|IsValid _physmat
    (:Is%20Valid
      (Variables|AttachableProjectile|SetSurfaceType (Variables|/Script/PhysicsCore.PhysicalMaterial|GetSurfaceType _physmat)))
    (:Is%20Not%20Valid
      _))
  (if _bblockinghit
    (Variables|AttachableProjectile|SetHitLocation _normal)
    (else
      (Variables|AttachableProjectile|SetHitLocation (Transformation|GetActorLocation))))
  (Collision|SetCollisionEnabled (Variables|Collision|GetCapsuleComponent) "NoCollision")
  (Transformation|SetActorLocationAndRotation (Variables|AttachableProjectile|GetHitLocation) (Math|Conversions|RotationFromXVector (Math|Vector|NegateVector (Variables|AttachableProjectile|GetHitNormal))))
  (Utilities|FlowControl|Sequence
    (:then_0
      (CallFunction|ClosePawnCollision))
    (:then_1
      (CallFunction|CallExplodeTrigger))
    (:then_2
      (bind _asbuilding_actor (Utilities|Casting|CastToBuildingActor _hitactor)
        (:then
          (EventDispatchers|BindEventtoOnDied _asbuilding_actor (EventDispatchers|CreateEvent :!event Custom|Event_BuildingActorDied))
          (Utilities|FlowControl|DoOnce false
            (:Completed
              (Transformation|AttachActorToComponent _hitcomponent "None" "KeepWorld" "KeepWorld" "KeepWorld" false))))
        (:CastFailed
          ;[1 1 1] Bind to vehicle destruction to restart simulation
          (bind _asfort_vehicle_interface (Utilities|Casting|CastToFortVehicleInterface _hitactor))
          (Game|BindEventtoOnDestroyed (Vehicle|GetActor _asfort_vehicle_interface) (delegate Custom|Event_OnVehicleDestroyed)))))
    (:then_3
      (CallFunction|LandSoundReplicate))))

(event Custom|Land%20Sound%20Replicate
  ;[1 1 1] OnStop - Bind destruction of building actor if relevant
  (CallFunction|PlayStickSound))

(event Custom|Event_OnVehicleDestroyed
  (params (DestroyedActor :type Actor%20Object%20Reference))
  ;[1 1 1] Bind to vehicle destruction to restart simulation
  (CallFunction|RestartSimulating))

(event Custom|BindToPlayerOrAIDeath
  ;[1 1 1] Restart simulation if player dies
  (EventDispatchers|BindEventtoOnDied (Variables|AttachableProjectile|GetStuckToPawn) (delegate Custom|PlayerDiedEvent)))

(event Custom|Event_BuildingActorDied
  (params (DamagedActor :type Actor%20Object%20Reference) (Damage :type Float%20%28single-precision%29) (InstigatedBy :type Controller%20Object%20Reference) (DamageCauser :type Actor%20Object%20Reference) (HitLocation :type Vector) (FHitComponent :type Primitive%20Component%20Object%20Reference) (BoneName :type Name) (Momentum :type Vector))
  ;[1 1 1] Restart simulation if building actor is destroyed
  (CallFunction|RestartSimulating))

(event EventOnAttachedToDied
  (params (DamagedActor :type Actor%20Object%20Reference) (Damage :type Float%20%28single-precision%29) (InstigatedBy :type Controller%20Object%20Reference) (DamageCauser :type Actor%20Object%20Reference) (HitLocation :type Vector :default "0, 0, 0") (FHitComponent :type Primitive%20Component%20Object%20Reference) (BoneName :type Name :default "None") (Momentum :type Vector :default "0, 0, 0"))
  ;[1 1 1] Ignore dead pawns previously attached to after pawn dies
  (Utilities|Set|Add (Variables|AttachableProjectile|GetIgnore) DamagedActor)
  (CallFunction|RestartSimulating))

(event Custom|PlayerDiedEvent
  (params (DamagedActor :type Actor%20Object%20Reference) (Damage :type Float%20%28single-precision%29) (InstigatedBy :type Controller%20Object%20Reference) (DamageCauser :type Actor%20Object%20Reference) (HitLocation :type Vector) (FHitComponent :type Primitive%20Component%20Object%20Reference) (BoneName :type Name) (Momentum :type Vector))
  ;[1 1 1] Restart simulation if player dies
  (CallFunction|RestartSimulating)
  (Utilities|Array|RemoveItem (Variables|Actor|GetTags) (Variables|Default|GetAttachedToPawn)))

(event EventEvent_Building_Actor_Destroyed
  ;[1 1 1] Restart simulation if building actor is destroyed
  (CallFunction|RestartSimulating))

(event Custom|Event_TriggerExplosion
  ;[1 1 1] Trigger Explosion
  (Utilities|FlowControl|DoOnce false
    (:Completed
      (Delivery|DoExplosion))))

(event EventOnExploded
  (params (HitActors :type Array%20of%20Actor%20Object%20References) (HitResults :type Array%20of%20Hit%20Result%20Structures))
  (bind _instigator (Variables|Actor|GetInstigator))
  (bind _returnvalue (Ability|GetFortAbilitySystemComponent _instigator))
  ;[1 1 1] Explosion
  (if (not (Variables|BGASpawning|GetSpawned))
    (Utilities|FlowControl|DoOnce false
      (:Completed
        (Utilities|FlowControl|Sequence
          (:then_0
            (Actor|SetLifeSpan 1.0)
            (CallFunction|PlayExplosionEffects))
          (:then_1
            (bind _outactors (Collision|SphereOverlapActors (Transformation|GetActorLocation) (ScalableFloat|GetValueatLevel (Variables|Explosion|GetExplosionRadius)) (Utilities|Array|MakeArray "ObjectTypeQuery6" "ObjectTypeQuery1" "ObjectTypeQuery5" "ObjectTypeQuery3" "ObjectTypeQuery2" "ObjectTypeQuery4" "ObjectTypeQuery7") "/Script/Engine.Actor" (Utilities|Array|MakeArray self)))
            (bind _result (CallFunction|FilterbyLOS _outactors))
            (bind _returnvalue_2 (Gameplay|GetAttachedtoActor))
            (Utilities|IsValid _returnvalue_2
              (:Is%20Valid
                (Utilities|Array|AddUnique _result _returnvalue_2) ; regardless of LOS allways hit the \nactor we're attached to.
              )
              (:Is%20Not%20Valid
                _))
            (Utilities|IsValid _returnvalue
              (:Is%20Valid
                (Variables|Default|SetDamageDealingAbilitySystemComponent _returnvalue))
              (:Is%20Not%20Valid
                (bind _asfort_ability_system_component (Utilities|Casting|CastToFortAbilitySystemComponent (Ability|GetAbilitySystemComponent self)))
                (Variables|Default|SetDamageDealingAbilitySystemComponent _asfort_ability_system_component)))
            (for _array_element _result
              (bind _asfort_pawn (Utilities|Casting|CastToFortPawn _array_element)
                (:then
                  (if (or (Math|Byte|NotEqual(Byte) (Team|GetActorTeam _asfort_pawn) (Variables|BGASpawning|GetTeam)) (Utilities|Equal(Object) _asfort_pawn (Variables|Actor|GetInstigator)))
                    _))
                (:CastFailed
                  _))
              (GameplayEffects|ApplyGameplayEffectToTarget (Variables|Default|GetDamageDealingAbilitySystemComponent) (Variables|Explosion|GetExplosionGE) (Ability|GetAbilitySystemComponent _array_element) 0.0 (CallFunction|MakeEffectContext)))))))))

(event Custom|PlayExplosionEffects
  (bind _returnvalue (Transformation|GetActorLocation))
  ;[1 1 1] Explosion
  (Audio|PlaySoundatLocation (Variables|Sound|GetExplodeSFX) _returnvalue "0, 0, 0" 1.0 1.0 0.0 0 0 self)
  (bind _returnvalue_2 (Niagara|SpawnSystematLocation (Variables|VFX|GetExplosionVFX) _returnvalue))
  (Rendering|SetReceivesDecals _returnvalue_2 false)
  (GameplayCue|ExecuteGameplayCue%28Burst%29Local self (Variables|Explosion|GetCameraShakeGC) (Ability|GameplayCue|MakeGameplayCueParameters 1.0 3.0))
  (Fortnite|ShowSoundIndicatorLocal 0 (Transformation|GetActorLocation) (Variables|/Script/Engine.SoundBase|GetMaxDistance (Variables|Sound|GetExplodeSFX)) "Gunshot")
  (Game|SpawnActorfromClass "/PrimalGameplay/Items/Consumables/CuddleFish/VFX/B_Decal_CuddleFish.B_Decal_CuddleFish_C" (Math|Conversions|ToTransform(Vector) (Transformation|GetActorLocation)) "Undefined" "MultiplyWithRoot" 0 0)
  (Networking|CanExecuteCosmeticEvents
    (:True
      (if (ScalableFloat|GetValueasBool (Variables|Default|GetHotfix_SpawnDespawnFish))
        (Utilities|FlowControl|DoOnce false
          (:Completed
            (Utilities|FlowControl|Delay 0.2)
            (bind _returnvalue_3 (Gameplay|GetAttachedtoActor))
            (Game|SpawnActorfromClass "/PrimalGameplay/Items/Consumables/CuddleFish/B_Athena_CuddleFish_Despawn.B_Athena_CuddleFish_Despawn_C" (Math|Transform|MakeTransform (Transformation|GetActorLocation) (Math|Rotator|MakeRotator 0.0 0.0 (.yaw (Transformation|GetActorRotation)))) "AlwaysSpawn" "MultiplyWithRoot" 0 _returnvalue_3 0)))))))

(event Custom|CallExplodeTrigger
  (bind _explosionplayrate (Variables|AttachableProjectile|GetExplosionPlayRate))
  (bind _returnvalue (ScalableFloat|GetScalableFloatMultiplier _explosionplayrate))
  ;[1 1 1] Trigger Explosion
  (Utilities|FlowControl|DoOnce false
    (:Completed
      (Utilities|Time|SetTimerbyEvent (EventDispatchers|CreateEvent :!event Custom|Event_TriggerExplosion) (/ (Animation|GetPlayLength (Variables|Anim|GetExplosionMontage)) _returnvalue))
      (Animation|Montage|PlayMontage (Variables|Default|GetSkeletalMesh) (Variables|Anim|GetExplosionMontage) _returnvalue))))

(event Custom|RestartSimulating
  ;[1 1 1] Restart Simulation
  (Variables|AttachableProjectile|SetbSimulationRunning true)
  (Networking|ForceNetUpdate))

(event EventOnStop
  (params (Hit :type Hit%20Result%20Structure%20%28by%20ref%29))
  (bind (_bblockinghit _binitialoverlap _time _distance _location _impactpoint _normal _impactnormal _physmat _hitactor) (Collision|BreakHitResult Hit))
  ;[1 1 1] Set Surface Type On Stop
  (|CallParentFunction Hit)
  (Utilities|IsValid _physmat
    (:Is%20Valid
      (Class|BPSurfaceTypeSoundComponent|SetSurfaceType (Variables|Default|GetBP_SurfaceTypeSoundComponent) (Variables|/Script/PhysicsCore.PhysicalMaterial|GetSurfaceType _physmat) _hitactor))))

(event Custom|AttachToActorVFX
  (bind _tags (Variables|Actor|GetTags))
  ;[1 1 1] Attach to Actor VFX
  (Utilities|Array|Clear _tags)
  (Utilities|Array|Add _tags (Variables|Default|GetAttachedToPawn))
  (Networking|CanExecuteCosmeticEvents
    (:True
      (Components|Activation|Activate (Variables|Default|GetAttachedVFX) true)
      (Audio|PlaySoundatLocation (Variables|Sound|GetAudioStuckToPlayer) (Transformation|GetWorldLocation (Variables|Default|GetSkeletalMesh))))))

(event Custom|SetHasSpawnedFromJump
  (bind _tags (Variables|Actor|GetTags))
  ;[1 1 1] Tags for animation states
  (Utilities|Array|Add _tags (Variables|Default|GetJumped))
  (Utilities|Array|RemoveItem _tags (Variables|Default|GetThrown)))

(event Custom|CheckVelocity
  (bind _tags (Variables|Actor|GetTags))
  (bind (_velocity_x _velocity_y _velocity_z) (Math|Vector|BreakVector (Variables|/Script/Engine.MovementComponent|GetVelocity_X (Variables|Movement|GetProjectileMovementComponent))))
  (if (< _velocity_z 0.0)
    (Utilities|FlowControl|DoOnce false
      (:Completed
        (Utilities|Array|Clear _tags)
        (Utilities|Array|Add _tags (Variables|Default|GetFalling))))))

(fn CheckForWater
  (params (Location :type Vector) (HitActor :type Object%20Reference))
  (bind _asfort_water_body_actor (Utilities|Casting|CastToFortWaterBodyActor HitActor)
    (:then
      (bind _outwatersurfaceinfo (WaterBody|GetWaterSurfaceInfoStruct _asfort_water_body_actor (Transformation|GetActorLocation) true false false))
      (bind _waterplanelocation (Utilities|Struct|BreakWaterSurfaceInfo _outwatersurfaceinfo))
      (bind (_waterplanelocation_x _waterplanelocation_y _waterplanelocation_z) (Math|Vector|BreakVector _waterplanelocation))
      (return (Math|Vector|MakeVector (.x Location) (.y Location) (+ _waterplanelocation_z (Variables|BGASpawning|GetSpawnPosZOffsetInWater)))))
    (:CastFailed
      (return Location))))

(fn FilterByLOS
  (params (Array :type Array%20of%20Actor%20Object%20References))
  (for _array_element Array
    (bind _asfort_pawn (Utilities|Casting|CastToFortPawn _array_element)
      (:then
        (bind _in_line_of_sight (Class|MissionBlueprintFunctionLibrary|CheckBasicVectortoActorLineOfSight (Transformation|GetActorLocation) _asfort_pawn "0, 0, 0" "TraceTypeQuery1" "None" 0.0 (LocalVariables|GetActorsToIgnore)))
        (if _in_line_of_sight
          _))
      (:CastFailed
        _))
    (Utilities|Array|Add (LocalVariables|GetResult) _array_element))
  (return (LocalVariables|GetResult)))

(fn GetClosestPoint
  (params (Component :type Primitive%20Component%20Object%20Reference) (StartingPoint :type Vector))
  (bind (_outpointonbody _returnvalue) (Collision|GetClosestPointonCollision Component StartingPoint "None"))
  (return (Math|Vector|SelectVector (+ (* (Math|Vector|Normalize (- StartingPoint _outpointonbody)) 18.0) _outpointonbody) (Transformation|GetActorLocation) (> _returnvalue 0.0))))

(fn HandleHitBuildingWall
  (params (Actor :type Actor%20Object%20Reference) (Component :type Primitive%20Component%20Object%20Reference))
  (bind _asbuilding_wall (Utilities|Casting|CastToBuildingWall Actor))
  (LocalVariables|SetLocalWall _asbuilding_wall)
  (bind _asstatic_mesh_component (Utilities|Casting|CastToStaticMeshComponent Component))
  (bind _returnvalue (BuildingWall|IsDoorComponent (LocalVariables|GetLocalWall) _asstatic_mesh_component))
  (if _returnvalue
    (Transformation|AttachActorToComponent Component "None" "KeepWorld" "KeepWorld" "KeepWorld" false)))

(fn HandleHits
  (params (Actor :type Actor%20Object%20Reference) (Component :type Primitive%20Component%20Object%20Reference))
  (Utilities|FlowControl|Sequence
    (:then_0
      (CallFunction|HandleHitBuildingWall Actor Component))
    (:then_1
      (CallFunction|HandleHitVehicle Actor Component))
    (:then_2
      (CallFunction|HandleHitSupplyDrop Actor Component))))

(fn HandleHitSupplyDrop
  (params (Actor :type Actor%20Object%20Reference) (Component :type Primitive%20Component%20Object%20Reference))
  (Utilities|FlowControl|Sequence
    (:then_0
      (Utilities|Casting|CastToAthenaSupplyDrop Actor)
      (Transformation|AttachActorToComponent Component "None" "KeepWorld" "KeepWorld" "KeepWorld"))
    (:then_1
      (Utilities|Casting|CastToAthenaSupplyDropBalloon Actor)
      (Transformation|AttachActorToComponent Component "None" "KeepWorld" "KeepWorld" "KeepWorld" false))))

(fn HandleHitVehicle
  (params (Actor :type Actor%20Object%20Reference) (Component :type Primitive%20Component%20Object%20Reference))
  (Utilities|Casting|CastToFortVehicleInterface Actor)
  (bind _asfort_character_vehicle_ostrich (Utilities|Casting|CastToFortCharacterVehicle_Ostrich Actor)
    (:then
      (Transformation|AttachActorToComponent (Variables|/Script/Engine.Character|GetMesh _asfort_character_vehicle_ostrich) "None" "KeepWorld" "KeepWorld" "KeepWorld" false))
    (:CastFailed
      (Transformation|AttachActorToComponent Component "None" "KeepWorld" "KeepWorld" "KeepWorld" false))) ; If we hit a mech, attach to mech mesh
)

(fn MakeEffectContext
  (bind _output_get (LocalVariables|SetEffectContext (GameplayEffects|MakeFortEffectContext (Variables|Default|GetDamageDealingAbilitySystemComponent) 0 0 self 0)))
  (Ability|EffectContext|AddHitResult _output_get (Collision|MakeHitResult false false 0.0 0.0 "0, 0, 0" "0, 0, 0" "0,0,1" "0,0,1" 0 0 0 "None" "None" 0 0 0 (Transformation|GetActorLocation) (Transformation|GetActorLocation)))
  (return (LocalVariables|GetEffectContext)))

(fn OnRep_bSimulationRunning
  (if (Variables|AttachableProjectile|GetbSimulationRunning)
    (Components|Movement|SetUpdatedComponent (Variables|Movement|GetProjectileMovementComponent) (Variables|Collision|GetCapsuleComponent))
    (Collision|SetCollisionEnabled (Variables|Collision|GetCapsuleComponent) "QueryOnly")
    (CallFunction|ReopenPawnCollision)))

(fn PlayStickSound
  (Audio|PlaySoundatLocation (select (Utilities|IsValid (Variables|AttachableProjectile|GetStuckToPawn) :!class /Script/Engine.KismetSystemLibrary) (Variables|Sound|GetAudioStuckToPlayer) (Variables|Sound|GetAudioStuckToEnvironment)) (Transformation|GetWorldLocation (Variables|Default|GetSkeletalMesh))))

(fn SpawnBGA
  (params (Normal :type Vector) (HitActor :type Actor%20Object%20Reference) (HitComponent :type Primitive%20Component%20Object%20Reference) (Location :type Vector) (FromWater :type Boolean) (TraceDifference :type Vector))
  (bind (_returnvalue_x _returnvalue_y _returnvalue_z) (Math|Vector|BreakVector (Transformation|GetActorLocation (Variables|BGASpawning|GetSpawnedBGA))))
  (if (Variables|BGASpawning|GetSpawned)
    (if (> (.z (Variables|GetLocation)) _returnvalue_z)
      (Actor|DestroyActor (Variables|BGASpawning|GetSpawnedBGA)))
    (else
      (Variables|BGASpawning|SetSpawned true)
      (Utilities|FlowControl|Sequence
        (:then_0
          (Actor|SetLifeSpan 2.0)
          (Collision|SetActorEnableCollision false)
          (Rendering|SetActorHiddenInGame true))
        (:then_1
          _))))
  (bind _returnvalue (Game|SpawnActorfromClass (Variables|BGASpawning|GetBGAToSpawn) (Math|Transform|MakeTransform (+ (CallFunction|CheckforWater (Variables|GetLocation) (Variables|GetHitActor)) (* (Utilities|MakeSureNormalIsNotZero (Variables|GetNormal)) (Variables|BGASpawning|GetSpawnPosOffsetAlongNormal))) (Math|Rotator|MakeRotfromZX (Utilities|MakeSureNormalIsNotZero (Variables|GetNormal)) (Math|Vector|ProjectVectorontoPlane (Variables|GetTraceDifference) (Utilities|MakeSureNormalIsNotZero (Variables|GetNormal)))) (Variables|BGASpawning|GetSpawnScale)) "Undefined" "SelectDefaultAtRuntime" (Variables|Actor|GetInstigator) (Variables|Actor|GetInstigator)))
  (bind _output_get (Variables|BGASpawning|SetSpawnedBGA _returnvalue))
  (bind _asbga_athena_cuddle_fish_jumper (Utilities|Casting|CastToBGA_Athena_CuddleFish_Jumper _output_get)
    (:then
      (Building|InitializeKismetSpawnedBuildingActor _asbga_athena_cuddle_fish_jumper 0 0 true 0 false)
      (bind _returnvalue_2 (CallFunction|GetOwnerWeapon))
      (CosmeticWraps|AttemptApplyOverrideWraptoActor _returnvalue_2 _asbga_athena_cuddle_fish_jumper)
      (Class|BGAAthenawithGravityParent|MulticastHandleBinding _asbga_athena_cuddle_fish_jumper (Variables|GetHitActor) (Variables|GetHitComponent)))
    (:CastFailed
      _))
  (Ability|SendGameplayEventtoActor (Actor|GetOwner) "(TagName=\"Event.Projectile.SpawnedBGA\")" (Utilities|Struct|MakeGameplayEventData "(TagName=\"\")" 0 0 _returnvalue))
  (Ability|SendGameplayEventtoActor (Game|GetInstigator) "(TagName=\"Abilities.Drops.AddToActorLimit.CuddleFish\")" (Utilities|Struct|MakeGameplayEventData "(TagName=\"\")" 0 (Variables|BGASpawning|GetSpawnedBGA)))
  (bind _asbga_athena_cuddle_fish_jumper_2 (Utilities|Casting|CastToBGA_Athena_CuddleFish_Jumper (Variables|BGASpawning|GetSpawnedBGA)))
  (if (Variables|BGASpawning|GetTeamHasBeenSet)
    (Class|BGAAthenaCuddleFishJumper|SetMyTeam _asbga_athena_cuddle_fish_jumper_2 (Variables|BGASpawning|GetTeam) :!class /PrimalGameplay/Items/Consumables/CuddleFish/BGA_Athena_CuddleFish_Jumper.BGA_Athena_CuddleFish_Jumper_C)
    (else
      _))
  (return (Variables|BGASpawning|GetSpawnedBGA)))

(fn ConstructionScript
  (|CallParentFunction))
