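; B_Decal_CuddleFish is a decal actor Blueprint for the CuddleFish consumable's visual effects that gives each placed decal a randomly varied look.
; Its construction script picks a random entry from the MaterialOptions array, creates a dynamic material instance from it, stores that as DecalMID, and assigns it to the Decal component.
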(fn ConstructionScript
  (bind _returnvalue (Rendering|Material|CreateDynamicMaterialInstance (Utilities|Array|RandomArrayItem (Variables|Default|GetMaterialOptions)) "None" "None" :!class /Script/Engine.KismetMaterialLibrary))
  (bind _output_get (Variables|Default|SetDecalMID _returnvalue))
  (Rendering|Components|Decal|SetDecalMaterial (Variables|Decal|GetDecal) _output_get))
