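# Translucent, lit deferred decal material for an iridescent paint splat in the CuddleFish consumable VFX, using splat masks and normals selected from a subdivided atlas by index parameters.
# A time-scrolled, flowmap-distorted noise drives a gradient lookup, blended over the decal color by a view-angle power mask.
# Decal lifetime erodes the mask and alters normal strength.
# World-position derivatives mask surfaces by decal orientation, bypassed on mobile.
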Material /PrimalGameplay/Items/Consumables/CuddleFish/VFX/Materials/M_Decal_IridescentSplat
graph {
 properties {blendMode=BLEND_Translucent, materialDomain=MD_DeferredDecal, pixelDepthOffsetMode=PDOM_Legacy, usedShadingModels=Default Lit}
 nodes [
  n1 = AppendVector(A=ScalarParameter() {defaultValue=4, parameterName=SubdivisionsU}, B=ScalarParameter() {defaultValue=1, parameterName=SubdivisionsV})
  n2 = ScalarParameter() {parameterName=IndexU}
  n3 = ScalarParameter() {parameterName=IndexV}
  n4 = Add(A=Divide(A=TextureCoordinate(), B=@n1), B=Divide(A=AppendVector(A=@n2, B=@n3), B=@n1))
  n5 = TextureSample(UVs=@n4) {samplerType=SAMPLERTYPE_Normal, texture=/Game/Athena/Items/Weapons/Prototype/Papaya/T_PaintSplats_Normal}
  n6 = OneMinus(None=DecalLifetimeOpacity().Opacity)
  LinearInterpolate(A=@n5.RGB, B=Multiply(A=@n5.RGB, B=VectorParameter() {parameterName=NormalOuterStength, defaultValue=<rgba 0.1 0.1 1 0>}.RGB), Alpha=@n6)
  n7 = Add(A=Multiply(A=Time(), B=ScalarParameter() {parameterName=Color Speed 2}), B=TextureCoordinate())
  n8 = Panner(Speed=ScalarParameter() {defaultValue=0.01, parameterName=Color Speed 3}) {speedY=0.1}
  n9 = ComponentMask(None=Multiply(
    A=TextureSample(UVs=@n8) {samplerType=SAMPLERTYPE_LinearColor, texture=/Game/Global/Textures/Default/Flowmaps/T_SwirlFlow_02}.RGB
    B=ScalarParameter() {defaultValue=1, parameterName=Distortion Strength}
   )) {g=true, r=true}
  n10 = TextureSample(UVs=Multiply(A=Add(A=@n7, B=@n9), B=ScalarParameter() {defaultValue=2, parameterName=Iridescence Tiling})) {texture=&2ndInfectedMapLinear}
  n11 = DotProduct(A=VertexNormalWS(), B=CameraVectorWS())
  n12 = TextureSample(UVs=@n4) {samplerType=SAMPLERTYPE_LinearColor, texture=/Game/Athena/Items/Weapons/Prototype/Papaya/T_PaintSplats_01}
  n13 = Power(Base=Multiply(A=Abs(None=@n11), B=@n12.G), Exp=ScalarParameter() {defaultValue=3, parameterName=Iridescence Power})
  LinearInterpolate(
   A=ScalarParameter() {defaultValue=0.5, parameterName=Metallic A}
   B=ScalarParameter() {defaultValue=0.5, parameterName=Metallic B}
   Alpha=Multiply(A=@n10.R, B=@n13)
  )
  n14 = Add(A=@n12.G, B=SphereMask(A=@n12.B, B=Constant(), Radius=@n6, Hardness=ScalarParameter() {defaultValue=0.2, parameterName=ErosionHardness}))
  n15 = Multiply(A=@n6) {constB=2}
  n16 = WorldPosition()
  n17 = DDX(Value=@n16.XYZ)
  n18 = DDY(Value=@n16.XYZ)
  n19 = DotProduct(A=Multiply(A=ObjectOrientation(), B=Constant() {r=-1}), B=CrossProduct(A=Normalize(VectorInput=@n17), B=Normalize(VectorInput=@n18)))
  n20 = Constant() {r=1}
  Multiply(A=Subtract(A=@n12.R, B=Multiply(A=@n14, B=@n15)), B=Saturate(None=ShadingPathSwitch(Default=@n19, Mobile=@n20)))
  n21 = VectorParameter() {parameterName=DecalColorA, defaultValue=<rgba 0.6111 0.1125 0.0127 0.3>}
  n22 = TextureObjectParameter() {parameterName=Iridescent Texture, texture=/Game/Weapons/WeaponSkins/Wraps/Textures/Iridescence/T_iridescent_gradient}
  n23 = Multiply(
   A=Desaturation(
    None=TextureSample(UVs=@n10.R, Tex=@n22) {texture=&T_iridescent_gradient}.RGB
    Fraction=ScalarParameter() {parameterName=Iridescence Desaturation Fraction}
   )
   B=VectorParameter() {parameterName=Iridescence, defaultValue=<rgba 1 1 1 0>}.RGB
  )
  Reroute(None=Multiply(A=@n21.A, B=LinearInterpolate(A=@n21.RGB, B=@n23, Alpha=@n13)))
  ScalarParameter() {defaultValue=0.7, parameterName=Specular}
  ScalarParameter() {defaultValue=0.25, parameterName=Roughness}
 ]
}
