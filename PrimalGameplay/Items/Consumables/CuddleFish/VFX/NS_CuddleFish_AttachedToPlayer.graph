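# Niagara water-splash effect for the CuddleFish consumable, named for the moment it attaches to the player, built from three burst-spawned sprite emitters.
# Smoke_plum puffs pale cyan, SubUV-animated mist outward from a small sphere, with drag and fast fade.
# Splashes throws rotating random-frame sprites upward that grow, fall and fade through a dynamic material parameter.
# Water_Droplets sprays an upward cone of tiny drops with partial inherited velocity, stretched by speed.
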NiagaraSystem /PrimalGameplay/Items/Consumables/CuddleFish/VFX/NS_CuddleFish_AttachedToPlayer
dependencies {
 usedDataInterfaces [&NiagaraDataInterfaceCurve, &NiagaraDataInterfaceVectorCurve]
 usedDynamicInputs []
 usedModules [&SystemState]
}
emitters {
 Smoke_plum {
  emitterUpdateScript [
   &EmitterState(Life Cycle Mode=::NewEnumerator1, Loop Behavior=::NewEnumerator1, Loop Delay=0.2, Loop Duration=1, UseLoopDelay=false)
   &SpawnBurst_Instantaneous(Spawn Count=5, Use Spawn Probability=false)
  ]
  particleSpawnScript [
   /Niagara/Modules/Spawn/Initialization/InitializeParticle(
    Color=<rgba 0.5627 0.9098 1 0.25>
    Lifetime=&UniformRangedFloat(Minimum=1, Maximum=1.5)
    Module.Write Color=true
    Module.Write Lifetime=true
    Module.Write Mass=true
    Module.Write Material Random=true
    Module.Write Position=false
    Module.Write Scale=false
    Module.Write SpriteRotation=true
    Module.Write SpriteSize=true
    Position=&SimulationPosition()
    Sprite Rotation=&UniformRangedFloat(Minimum=50, Maximum=100)
    Sprite Size=&Vector2DFromFloat(Value=&UniformRangedFloat(Minimum=80, Maximum=125))
   )
   &AddVelocity(Velocity=&UniformRangedVector(Minimum=<-10 0 -15>, Maximum=<25 0 15>)) {enabled=false}
   &CameraOffset(Camera Offset Amount=&UniformRangedFloat(Minimum=15, Maximum=25)) {enabled=false}
   &SphereLocation(
    MaskSpawnGroup=false
    Override Seed=false
    Sphere Radius=1
    Surface Only=false
    UseNonUniformScale=false
    UseOffset=false
    UseSphereOrientation=false
   )
   &AddVelocity(Velocity=&Multiply_VectorByFloat(Vector=$Output.SphereLocation.SphereNormal, Float=&RandomRangeFloat(Minimum=50, Maximum=150))) {module=AddVelocity001}
  ]
  particleUpdateScript [
   &ParticleState()
   &Drag(Drag=&UniformRangedFloat(Minimum=1, Maximum=3), Use Linear Drag=true, Use Rotational Drag=false)
   &GravityForce(Gravity=<0 0 -10>)
   &SolveForcesAndVelocity(Clamp Velocity=false, Limit Acceleration=false)
   /Niagara/Modules/Update/SubUV/SubUVAnimation(
    Number Of Frames=64
    SubUV Lookup Index=&FloatFromCurve(FloatCurve=DataInterface() {propertyValues {Curve {keys [
         {arriveTangent=0.3268, interpMode=RCIM_Cubic, leaveTangent=0.3268, tangentMode=RCTM_User, time=0, value=0.6}
         {arriveTangent=0.0289, interpMode=RCIM_Cubic, leaveTangent=0.0289, tangentMode=RCTM_User, time=1.0011, value=0.985}
        ]}}})
   )
   &ScaleColor(
    Scale Alpha=&FloatFromCurve(FloatCurve=DataInterface() {propertyValues {Curve {keys [
         {interpMode=RCIM_Cubic, leaveTangent=51.4821, tangentMode=RCTM_Break, time=0, value=0}
         {arriveTangent=-0.551, interpMode=RCIM_Cubic, leaveTangent=-0.551, tangentMode=RCTM_Break, time=0.0608, value=1.0115}
         {arriveTangent=-1.321, interpMode=RCIM_Cubic, leaveTangent=-1.321, tangentMode=RCTM_User, time=0.2408, value=0.4168}
         {arriveTangent=0.0661, interpMode=RCIM_Cubic, leaveTangent=0.0661, tangentMode=RCTM_User, time=1.0011, value=0.002}
        ]}}})
    ScaleA=true
    ScaleRGB=true
   )
   &ScaleSpriteSize(Scale Factor=&Vector2DFromFloat(Value=&FloatFromCurve(
      FloatCurve=DataInterface() {propertyValues {Curve {keys [
          {arriveTangent=7.3273, interpMode=RCIM_Cubic, leaveTangent=7.3273, tangentMode=RCTM_User, time=0, value=0}
          <at 0.131 0.6357>
          <at 1 1 RCIM_Cubic>
         ]}}}
      Scale Curve=2
     )))
  ]
  rendererClasses [&NiagaraSpriteRendererProperties]
  renderers [{rendererClass=&NiagaraSpriteRendererProperties}]
 }
 Splashes {
  emitterUpdateScript [
   &EmitterState(Life Cycle Mode=::NewEnumerator1, Loop Behavior=::NewEnumerator1, Loop Duration=1, UseLoopDelay=false)
   &EmitterScalability(
    Module.bApplySpawnCountScale=false
    Module.bApplySpawnCountScaleByDistanceFraction=false
    Module.bUseMaxDistance=false
    Module.bUseMinDistance=false
   ) {enabled=false}
   &EmitterLifeCycle() {enabled=false}
   &SpawnBurst_Instantaneous(Spawn Count=5, Use Spawn Probability=false)
  ]
  particleSpawnScript [
   &InitializeParticle(
    Color=&UniformRangedLinearColor(Minimum=<rgba 0 0 0 1>, Maximum=<rgba 5 5 5 1>)
    Lifetime=&UniformRangedFloat(Minimum=0.4, Maximum=1)
    Module.Write Color=true
    Module.Write Lifetime=true
    Module.Write Mass=true
    Module.Write Material Random=true
    Module.Write Position=true
    Module.Write Scale=false
    Module.Write SpriteRotation=true
    Module.Write SpriteSize=true
    Position=&SimulationPosition()
    Sprite Rotation=&UniformRangedFloat(Minimum=0, Maximum=360)
    Sprite Size=&Vector2DFromFloat(Value=&UniformRangedFloat(Minimum=20, Maximum=35))
   )
   /NiagaraFN/Sprite/FN_RandomizeSpriteUV()
   &SubUVAnimation(SubUV Lookup Index=&UniformRangedFloat(Minimum=0, Maximum=1))
   &AddVelocity(Coordinate Space=::World, Velocity=&UniformRangedVector(Minimum=<-70 -70 150>, Maximum=<70 70 300>))
   &Color(Color=&UniformRangedLinearColor(Minimum=<rgba 0.7478 0.8391 1 0.5>, Maximum=<rgba 1.6508 1.899 2 1>))
  ]
  particleUpdateScript [
   &ParticleState()
   &GravityForce()
   &SolveForcesAndVelocity(Clamp Velocity=false, Limit Acceleration=false)
   &ScaleSpriteSize(Scale Factor=&Multiply_Vector2(
     A=<2 2>
     B=&Vector2DFromFloat(Value=&Multiply_Float(
       A=1
       B=&FloatFromCurve(FloatCurve=DataInterface() {propertyValues {Curve {keys [
            {arriveTangent=2.0364, interpMode=RCIM_Cubic, leaveTangent=2.0364, tangentMode=RCTM_User, time=0.0023, value=0.3733}
            {arriveTangent=0.7301, interpMode=RCIM_Cubic, leaveTangent=0.7301, time=0.3082, value=0.6733}
            <at 0.3112 0.6755>
            {arriveTangent=0.617, interpMode=RCIM_Cubic, leaveTangent=0.617, tangentMode=RCTM_User, time=0.6894, value=0.875}
            <at 0.994 1.0109>
           ]}}})
      ))
    ))
   &ScaleColor(
    Scale Alpha=&FloatFromCurve(FloatCurve=DataInterface() {propertyValues {Curve {keys [
         {interpMode=RCIM_Cubic, tangentMode=RCTM_User, time=-0.0007, value=-0.0008}
         {arriveTangent=0.0027, interpMode=RCIM_Cubic, leaveTangent=0.0027, tangentMode=RCTM_User, time=0.0821, value=0.9831}
         {arriveTangent=-1.0675, interpMode=RCIM_Cubic, leaveTangent=-1.0675, tangentMode=RCTM_User, time=0.8646, value=0.7677}
         <at 1.0011 0.002>
        ]}}})
    Scale RGB=&ScaleFloatByCurve(
     InitialValue:Vector3f=<1 1 1>
     ScaleCurve:NiagaraDataInterfaceVectorCurve=DataInterface() {propertyValues {
       XCurve {keys [<at 0 0>,<at 0.3112 0.6225>,<at 0.5 1>,<at 1 0>]}
       YCurve {keys [<at 0 0>,<at 0.3112 0.6225>,<at 0.5 1>,<at 1 0>]}
       ZCurve {keys [<at 0 0>,<at 0.3112 0.6225>,<at 0.5 1>,<at 1 0>]}
      }}
     CurveIndex:NiagaraFloat=$Particles.NormalizedAge
    )
    ScaleA=true
    ScaleRGB=true
   )
   &DynamicMaterialParameters(
    Index 0 Param 1=&FloatFromCurve(FloatCurve=DataInterface() {propertyValues {Curve {keys [
         {interpMode=RCIM_Cubic, tangentMode=RCTM_User, time=-0.0046, value=1.0113}
         {arriveTangent=-1.0036, interpMode=RCIM_Cubic, leaveTangent=-1.0036, tangentMode=RCTM_User, time=0.705, value=0.8464}
         <at 1.0011 0.002>
        ]}}})
    Module.Param0WriteEnabled=true
    Module.Param1WriteEnabled=true
    Module.Param2WriteEnabled=true
    Module.Param3WriteEnabled=true
   )
   &SpriteRotationRate(Rotation Rate=&RandomRangeFloat(Minimum=-100, Maximum=100))
  ]
  rendererClasses [&NiagaraSpriteRendererProperties]
  renderers [{rendererClass=&NiagaraSpriteRendererProperties}]
 }
 Water_Droplets {
  emitterUpdateScript [
   &EmitterState(Life Cycle Mode=::NewEnumerator1, Loop Behavior=::NewEnumerator1, Loop Duration=1, UseLoopDelay=false)
   &EmitterScalability(
    Module.bApplySpawnCountScale=false
    Module.bApplySpawnCountScaleByDistanceFraction=false
    Module.bUseMaxDistance=false
    Module.bUseMinDistance=false
   ) {enabled=false}
   &EmitterLifeCycle() {enabled=false}
   &SpawnBurst_Instantaneous(Spawn Count=35, Use Spawn Probability=false)
   &SpawnRate(SpawnRate=35, Use Spawn Probability=false) {enabled=false}
  ]
  particleSpawnScript [
   &InitializeParticle(
    Color=<rgba 3 3 3 1>
    Lifetime=&UniformRangedFloat(Minimum=0.4, Maximum=1)
    Module.Write Color=true
    Module.Write Lifetime=true
    Module.Write Mass=true
    Module.Write Material Random=true
    Module.Write Position=false
    Module.Write Scale=false
    Module.Write SpriteRotation=true
    Module.Write SpriteSize=true
    Position=&SimulationPosition()
    Sprite Rotation=&UniformRangedFloat(Minimum=0, Maximum=0)
    Sprite Size=&Vector2DFromFloat(Value=&UniformRangedFloat(Minimum=1, Maximum=2))
   )
   /Niagara/Modules/Spawn/Location/SphereLocation(MaskSpawnGroup=false, Module.Override Randomness=false, Non Uniform Scale=<1 1 0.2>, Sphere Radius=5, Surface Only=false)
   &AddVelocityInCone(
    Cone Angle=35
    Cone Axis=<0 0 1>
    Cone Axis Coordinate Space=::World
    Module.Use Velocity Falloff On Cone Axis=false
    Velocity Strength=&RandomRangeFloat(Minimum=150, Maximum=450)
   )
  ]
  particleUpdateScript [
   &ParticleState()
   &GravityForce()
   &InheritVelocity(Module.Limit Speed=true, Velocity Scale=<-0.2 -0.2 0.1>)
   &Drag(Drag=0.5, Use Linear Drag=true, Use Rotational Drag=false)
   &SolveForcesAndVelocity(Clamp Velocity=false, Limit Acceleration=false)
   &ScaleColor(
    Scale Alpha=&FloatFromCurve(FloatCurve=DataInterface() {propertyValues {Curve {keys [
         {interpMode=RCIM_Cubic, tangentMode=RCTM_User, time=0, value=0}
         <at 0.1037 1>
         {interpMode=RCIM_Cubic, tangentMode=RCTM_User, time=0.7394, value=0.9972}
         <at 1 0>
        ]}}})
    ScaleA=true
    ScaleRGB=true
   )
   &ScaleSpriteSizeBySpeed(Max Scale Factor=<2 15>)
  ]
  rendererClasses [&NiagaraSpriteRendererProperties]
  renderers [{rendererClass=&NiagaraSpriteRendererProperties}]
 }
}
summary {
 systemName=NS_CuddleFish_AttachedToPlayer
 emitters [
  {emitterName=Smoke_plum, rendererClasses [&NiagaraSpriteRendererProperties]}
  {emitterName=Splashes, rendererClasses [&NiagaraSpriteRendererProperties]}
  {emitterName=Water_Droplets, rendererClasses [&NiagaraSpriteRendererProperties]}
 ]
 userVariables []
}
